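# Material function for the middle layer of the Megascans surface blend material that outputs a camera-distance-based tessellation multiplier.
# It measures the length of the camera-to-pixel world position vector, subtracts an offset and divides by a fade range.
# The saturated, inverted result scales the multiplier, so tessellation is full up close and falls linearly to zero with distance.
# Offset, fade and multiplier share one vector parameter in the Displacement group.
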MaterialFunction /Bridge/MSPresets/M_MS_SurfaceBlend_Material/LayerFunctions/Middle/MF_MiddleLayerTessellation
graph {
 properties {bExposeToLibrary=true}
 nodes [
  n1 = Distance(A=Subtract(A=CameraPositionWS(), B=WorldPosition().XYZ), B=Constant())
  n2 = VectorParameter() {
   group=06 - Displacement
   parameterName=Middle Layer Tessellation Controls
   sortPriority=2
   channelNames {a=Useless, b=Tessellation Multiplier, g=Distance Fade, r=Distance Offset}
   defaultValue=<rgba 0 500 16 1>
  }
  n3 = Saturate(None=Divide(A=Subtract(A=@n1, B=@n2.Distance Offset), B=@n2.Distance Fade))
  FunctionOutput(None=Multiply(A=OneMinus(None=@n3), B=@n2.Tessellation Multiplier))
 ]
}
